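; Prototype visor test actor: an interactable, destructible button that temporarily switches the player's view to a visor camera.
; Interaction locks the button, then the server blends the first player state's controller to the camera, waits, blends back and unlocks.
; PingPlayersInArea, uncalled here, debug-draws a sphere at EyeSocket and sends overlapping pawns an ability set and gameplay event, with blank tags.
; Death plays C4 effects and hides the mesh on remotes; an OnRep toggles the light.
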(event Interaction|EventBlueprintOnInteract
  (params (InteractingPawn :type Fort%20Pawn%20Object%20Reference) (InteractionBeingAttempted :type EInteractionBeingAttempted%20Enum :default "FirstInteraction"))
  ; Button Interaction Logic
  (Variables|Default|SetButtonCanBePressed false)
  (CallFunction|ButtonPressedEvent)
  (Development|PrintString "Button Pressed - Event on Server" true true "(R=0.000000,G=0.660000,B=1.000000,A=1.000000)" 5.0))

(event Custom|ButtonPressedEvent
  ; Button Interaction Event - What Happens?
  (Networking|SwitchHasAuthority
    (:Authority
      (Game|Player|SetViewTargetwithBlend (PlayerState|GetPlayerController (Game|GetPlayerState 0)) (Variables|Default|GetvisorCamera) (Variables|Default|GetInBlendTime) "VTBlend_EaseInOut" 1.0 false)
      (Utilities|FlowControl|Delay (Variables|Default|GetViewDurationTime))
      (Game|Player|SetViewTargetwithBlend (PlayerState|GetPlayerController (Game|GetPlayerState 0)) (Game|GetPlayerState 0) (Variables|Default|GetOutBlendTime) "VTBlend_EaseInOut" 1.0 false)
      (Variables|Default|SetButtonCanBePressed true))) ; authority = server, Remote = Client
)

(event Custom|PingPlayersInArea
  (bind _staticmesh (Variables|Default|GetStaticMesh))
  (bind _returnvalue (Transformation|GetSocketLocation _staticmesh "EyeSocket"))
  (bind _visualpingradius (Variables|Default|GetVisualPingRadius))
  (bind _returnvalue_2 (ScalableFloat|GetValueatLevel _visualpingradius))
  ;[0.02283 0.04199 0.15 0.5] Ping Players in the area every x seconds
  (Rendering|Debug|DrawDebugSphere _returnvalue _returnvalue_2 12 "(R=1.000000,G=0.000000,B=0.006200,A=1.000000)" 3.0 1e+01)
  (bind _outactors (Collision|SphereOverlapActors _returnvalue _returnvalue_2 (Variables|Default|GetObjectTypes) "/Script/FortniteGame.FortPlayerPawnAthena" (Utilities|Array|MakeArray self)))
  (for _array_element _outactors
    (bind _asfort_player_controller_athena (Utilities|Casting|CastToFortPlayerControllerAthena (Fortnite|GetFortPlayerControllerfromActor _array_element)))
    (if (not (Ability|GameplayAbility|HasAnyAbilitieswithAssetTag (Ability|GetAbilitySystemComponent _asfort_player_controller_athena) "(TagName=\"\")"))
      (Ability|EquipFortAbilitySet _asfort_player_controller_athena 0 _array_element)
      (Utilities|FlowControl|DelayUntilNextTick)
      (else
        _))
    (Ability|SendGameplayEventtoActor _array_element "(TagName=\"\")" (Utilities|Struct|MakeGameplayEventData "(TagName=\"\")" self))))

(event EventOnDeathServer
  (params (Damage :type Float%20%28single-precision%29) (DamageTags :type Gameplay%20Tag%20Container%20Structure%20%28by%20ref%29) (Momentum :type Vector :default "0, 0, 0") (HitInfo :type Hit%20Result%20Structure%20%28by%20ref%29) (InstigatedBy :type Controller%20Object%20Reference) (DamageCauser :type Actor%20Object%20Reference) (EffectContext :type Gameplay%20Effect%20Context%20Handle%20Structure))
  ; Button Destroyed Logic
  (CallFunction|PlayDeathFXAndSounds))

(event Custom|PlayDeathFXAndSounds
  ; Button Destroyed Logic
  (Networking|SwitchHasAuthority
    (:Remote
      (Niagara|SpawnSystematLocation "/Game/Athena/Items/Consumables/C4/NS_C4_Explosion_Athena.NS_C4_Explosion_Athena" (Transformation|GetActorLocation self)) ; Spawn a new niagara particle effect at this location
      (Audio|SpawnSoundatLocation "/Game/Athena/Sounds/Items/C4/C4_Explosion_Cue.C4_Explosion_Cue" (Transformation|GetActorLocation self)) ; a death sound
      (Rendering|SetVisibility (Variables|Default|GetStaticMesh) false false) ; this hides the mesh on the clients as its destroyed
    )))

(fn BlueprintCanInteract
  (params (InteractingPawn :type Fort%20Pawn%20Object%20Reference) (InteractionBeingAttempted :type EInteractionBeingAttempted%20Enum :default "FirstInteraction") (InteractionType :type TInteractionType%20Enum :default "IT_NoInteraction"))
  (return (Variables|Default|GetButtonCanBePressed)))

(fn BlueprintGetInteractionString
  (params (InteractingPawn :type Fort%20Pawn%20Object%20Reference) (InteractionBeingAttempted :type EInteractionBeingAttempted%20Enum :default "FirstInteraction"))
  (return (Variables|Default|GetButtonInteractionText)))

(fn OnRep_ButtonIsPressed
  (if (Variables|Default|GetButtonIsPressed)
    (Rendering|SetVisibility (Variables|Default|GetPointLight) true false)
    (else
      (Rendering|SetVisibility (Variables|Default|GetPointLight) false false))))

(fn ConstructionScript)
