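; Prototype visor camera actor that steers its camera component's view with the player's mouse cursor.
; Every tick it reads player controller 0's mouse screen position and uses it directly as a target rotation, with X as yaw, negated Y as pitch and zero roll.
; RInterpTo eases the camera's world rotation toward that target, with no scaling or clamping of the pixel values.
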(event EventTick
  (params (DeltaSeconds :type Float%20%28single-precision%29))
  (bind (_locationx _locationy) (Game|Player|GetMousePosition (Game|GetPlayerController 0)))
  (Transformation|SetWorldRotation (Variables|CameraActor|GetCameraComponent) (Math|Interpolation|RInterpTo (Transformation|GetWorldRotation (Variables|CameraActor|GetCameraComponent)) (Math|Rotator|MakeRotator 0.0 (* _locationy -1.0) _locationx) (Utilities|Time|GetWorldDeltaSeconds) 1e+01) false false))

(fn ConstructionScript)
